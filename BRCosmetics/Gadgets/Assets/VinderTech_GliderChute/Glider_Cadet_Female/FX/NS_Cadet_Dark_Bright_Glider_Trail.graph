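# Niagara glider trail system for the Cadet Dark Bright glider, combining orange trails with white speed lines.
# AnimTrail draws two bone-pair ribbons from the user skeletal mesh; StaticHandSprites001 emits orange sprites at two mirrored fixed offsets.
# SpeedLine spawns speed-stretched torus sprites when fully deployed, faded by ForwardDot in gameplay.
# A system integrator drives gameplay ribbon alpha and hand-sprite rate, while front-end switches reverse velocities and override intensities.
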NiagaraSystem /BRCosmetics/Gadgets/Assets/VinderTech_GliderChute/Glider_Cadet_Female/FX/NS_Cadet_Dark_Bright_Glider_Trail
dependencies {
 usedDataInterfaces [&NiagaraDataInterfaceVector2DCurve, &NiagaraDataInterfaceCurve, &NiagaraDataInterfaceColorCurve]
 usedDynamicInputs []
 usedModules [
  &self:SystemSpawnScript.SystemScriptSource.SystemScriptGraph.NiagaraNodeAssignment_0.SetVariables_#1
  &SystemState
  &FN_Integrator
 ]
}
emitters {
 AnimTrail {
  emitterUpdateScript [
   &EmitterState()
   /Game/Effects/Niagara/Modules/AnimTrail/FN_SpawnAnimTrail/FN_SpawnAnimTrail(Number of Anim Trails=2, SpawnRate=50)
   /Game/Effects/Niagara/Modules/AnimTrail/FN_SpawnAnimTrail/FN_SpawnAnimTrail_v6/FN_SpawnAnimTrail(Number of Anim Trails=2) {enabled=false, module=FN_SpawnAnimTrail001}
  ]
  particleSpawnScript [
   /Game/Effects/Niagara/Modules/AnimTrail/FN_InitializeAnimTrail/FN_InitializeAnimTrailBoneIndecies(Second Bone or Socket Index=1)
   &FN_InitializeAnimTrailBoneIndecies(AnimTrail Selection=1, First Bone or Socket Index=2, Second Bone or Socket Index=3) {module=FN_InitializeAnimTrailBoneIndecies001}
   /Game/Effects/Niagara/Modules/AnimTrail/FN_InitializeAnimTrail/FN_InitializeAnimTrail(AnimTrailWidth=1, Color=<rgba 0.6667 0.1135 0.5848 1>, Lifetime=0.16) {enabled=false}
   /SetVariables(Particles.AlphaFadeIn:NiagaraFloat=/Niagara/DynamicInputs/Math/Min_Float(A:NiagaraFloat=&Divide_Float(A=$Emitter.Age, B=1), B:NiagaraFloat=1)) {isSetParameters=true}
   /Game/Effects/Niagara/Modules/AnimTrail/FN_InitializeAnimTrail/FN_InitializeAnimTrail_v5/FN_InitializeAnimTrail(
    Color=&MakeLinearColor(
     R:NiagaraFloat=1
     G:NiagaraFloat=1
     B:NiagaraFloat=1
     A:NiagaraFloat=&Multiply_Float(A=$Particles.AlphaFadeIn, B=0.5)
    )
    Lifetime=&MakeCustomFloatFromBool(Bool=$Particles.Locked, True Float=100, False Float=0.8)
    Position Offset=&FN_FrontEndVector3Modifier(Gameplay=<0 -20 0>, Front End=<0 20 0>)
   ) {module=FN_InitializeAnimTrail001}
   &AddVelocity(Velocity=&FN_FrontEndVector3Modifier(Gameplay=<0 400 0>, Front End=<0 -400 0>)) {module=AddVelocity003}
   &AddVelocity(Velocity=&Multiply_VectorByFloat(Vector=<0 400 0>, Float=&MakeCustomFloatFromBool(Bool=$User.bIsFrontEnd, True Float=-1, False Float=1))) {enabled=false, module=AddVelocity002}
   &AddVelocity(Velocity=&MakeCustomVectorFromBool(Bool=$User.IsDebugInEditor, True Vector=<0 -2000 1>, False Vector=<1 0 0>)) {enabled=false, module=AddVelocity001}
   /SetVariables(Particles.Color=&LerpLinearColors(
     StartColor=&ColorFromCurve(
      ColorCurve=DataInterface() {propertyValues {
        AlphaCurve {keys [<at 0 1 RCIM_Cubic>]}
        BlueCurve {keys [<at 0.1027 0>]}
        GreenCurve {keys [<at 0.1027 0.4745>]}
        RedCurve {keys [<at 0.1027 0.9725>]}
       }}
      CurveIndex=&RandomRangeFloat(Minimum=0, Maximum=1)
     )
     EndColor=&ColorFromCurve(
      ColorCurve=DataInterface() {propertyValues {
        AlphaCurve {keys [<at 0 1 RCIM_Cubic>]}
        BlueCurve {keys [<at 0.0929 0>]}
        GreenCurve {keys [<at 0.0929 0.4745>]}
        RedCurve {keys [<at 0.0929 0.9725>]}
       }}
      CurveIndex=&RandomRangeFloat(Minimum=0, Maximum=1)
     )
     LerpFactor=&MakeFloatFromInt(Int=$Particles.SpawnGroup)
    )) {isSetParameters=true}
  ]
  particleUpdateScript [
   /Game/Effects/Niagara/Modules/AnimTrail/FN_UpdateAnimTrail/FN_UpdateAnimTrail(New float=1) {enabled=false}
   /Game/Effects/Niagara/Modules/AnimTrail/FN_UpdateAnimTrail/FN_UpdateAnimTrail_v5/FN_UpdateAnimTrail(
    Locked Particle Color:LinearColor=&MultiplyLinearColor(Linear Color=$Particles.Initial.Color, Scale Factor=<1 1 1 1>)
    Locked Particle RibbonWidth=1
    SkeletalMesh=$User.SkeletalMesh
   ) {module=FN_UpdateAnimTrail001}
   &SolveForcesAndVelocity(Clamp Velocity=false, Limit Acceleration=false)
   &Color(
    Color=$Particles.Color
    Scale Alpha=&MakeCustomFloatFromBool(
     Bool=$User.bIsFullyDeployed
     True Float=&FN_FrontEndFloatModifier(
      Gameplay=&ClampFloat(Float=&Multiply_Float(A=2, B=$System.FN_Integrator001.Value))
      Front End=1
     )
     False Float=0
    )
   )
  ]
  rendererClasses [&NiagaraRibbonRendererProperties]
  renderers [{rendererClass=&NiagaraRibbonRendererProperties}]
 }
 SpeedLine {
  emitterUpdateScript [
   &EmitterState()
   &SpawnRate(
    Spawn Probability=&MakeCustomFloatFromBool(Bool=$User.bIsFullyDeployed, True Float=1, False Float=0)
    SpawnRate=&MakeCustomFloatFromBool(Bool=$User.bIsFrontEndPreview, True Float=50, False Float=90)
    Use Spawn Probability=true
   ) {module=SpawnRate001}
  ]
  particleSpawnScript [
   &InitializeParticle(
    Color=&MakeLinearColorFromVectorAndFloat(
     "Vector (RGB)"=<1 1 1>
     "Float (Alpha)"=&Multiply_Float(
      A=&UniformRangedFloat(Minimum=0.1, Maximum=0.18)
      B=&MakeCustomFloatFromBool(Bool=$User.bIsFrontEndPreview, True Float=0.5, False Float=1)
     )
    )
    Lifetime Max=0.16
    Lifetime Min=0.1
    Lifetime Mode=::NewEnumerator1
    Sprite Size Max=<3 10>
    Sprite Size Min=<2 2>
    Sprite Size Mode=::NewEnumerator2
    UsePositionOffset=false
    Write Material Random=true
   )
   &TorusLocation(
    Align Torus With Axis=true
    Handle Radius=0
    Large Radius:NiagaraFloat=&RandomRangeFloat(Minimum=80, Maximum=150)
    MaskSpawnGroup=false
    Module.Override Randomness=false
    Offset=<0 -180 0>
    Torus Axis=<0 1 0>
    Torus Origin:NiagaraPosition=&Add_Vector(
     A=Linked() {linkedVariable {name=Particles.Position, type=NiagaraPosition}}
     B=&MakeVector(X=0, Y=&MakeCustomFloatFromBool(Bool=$User.bIsFrontEndPreview, True Float=200, False Float=0), Z=0)
    )
   )
   &AddVelocityFromPoint(
    Module.Use Falloff Distance=false
    Origin Offset=<0 -500 0>
    Velocity Strength=&Multiply_Float(A=600, B=&MakeCustomFloatFromBool(Bool=$User.bIsFrontEndPreview, True Float=-1, False Float=1))
   )
   &AddVelocity(Velocity=&Multiply_Vector(
     A=&RandomRangeVector(Minimum=<0 300 0>, Maximum=<0 700 0>)
     B=&MakeVector(X=1, Y=&MakeCustomFloatFromBool(Bool=$User.bIsFrontEndPreview, True Float=-1, False Float=1), Z=1)
    ))
  ]
  particleUpdateScript [
   &SolveForcesAndVelocity(Clamp Velocity=false, Limit Acceleration=false)
   &ParticleState()
   &ScaleSpriteSize(Scale Factor=&Vector2DFromCurve(
     Vector2Curve=DataInterface() {propertyValues {XCurve {keys [<at 0 0.5 RCIM_Cubic>,<at 0.5 1 RCIM_Cubic>]}, YCurve {keys [<at 0 0 RCIM_Cubic>,<at 0.5 1 RCIM_Cubic>]}}}
     ScaleCurve=&Vector2DFromFloat(Value=1)
    ))
   &ScaleSpriteSizeBySpeed(Max Scale Factor=<1 50>, Min Scale Factor=<0.01 0.05>, Velocity Threshold=1)
   &ScaleColor(
    Scale Alpha=&FloatFromCurve(
     FloatCurve=DataInterface() {propertyValues {Curve {keys [
         <at 0 0 RCIM_Cubic>
         <at 0.3759 1.0207>
         {arriveTangent=-2.221, interpMode=RCIM_Cubic, leaveTangent=-2.221, tangentMode=RCTM_User, time=1, value=0}
        ]}}}
     Scale Curve=&MakeCustomFloatFromBool(
      Bool=$User.bIsFrontEndPreview
      True Float=&MakeCustomFloatFromBool(Bool=$User.bIsFullyDeployed, True Float=1, False Float=0)
      False Float=&Lerp_Float(A=0, B=1, Alpha=$User.ForwardDot)
     )
    )
    Scale RGB=&VectorFromFloat(Value=1)
    ScaleA=true
    ScaleRGB=true
   )
  ]
  rendererClasses [&NiagaraSpriteRendererProperties]
  renderers [{rendererClass=&NiagaraSpriteRendererProperties}]
 }
 StaticHandSprites001 {
  emitterSpawnScript [&FN_PreserveSkydiveUserVariables() {module=PreserveSkydiveUserVariables}]
  emitterUpdateScript [
   &EmitterScalability(
    MaxDistance=2000
    Module.bApplySpawnCountScale=false
    Module.bApplySpawnCountScaleByDistanceFraction=false
    Module.bUseMaxDistance=true
    Module.bUseMinDistance=true
    SpawnCountScaleByDistanceFraction=DataInterface() {propertyValues {Curve {keys [<at -0.0098 1>,<at 0.9896 0>]}}}
   ) {enabled=false}
   &EmitterLifeCycle(NextLoopDuration=1) {enabled=false}
   &EmitterState(Enable Distance Culling=true, MaxDistance=2000, Module.bUseMaxDistance=true, Module.bUseMinDistance=false, Scalability Mode=::NewEnumerator1)
   &SpawnRate(
    SpawnRate=&MakeCustomFloatFromBool(
     Bool=$User.bIsFrontEndPreview
     True Float=90
     False Float=&Lerp_Float(A=0, B=90, Alpha=$System.FN_Integrator001.Value)
    )
    Use Spawn Probability=false
   ) {module=SpawnRate001}
  ]
  particleSpawnScript [
   &FN_SetSpawnGroup(Number of Spawn Groups=2) {module=FN_SetSpawnGroup001}
   /Niagara/Modules/Spawn/Initialization/InitializeParticle(
    Lifetime=&MakeCustomFloatFromBool(Bool=$User.bIsFullyDeployed, True Float=0.3, False Float=0)
    Module.Write Color=false
    Module.Write Lifetime=true
    Module.Write Mass=false
    Module.Write Material Random=true
    Module.Write Position=false
    Module.Write Scale=false
    Module.Write SpriteRotation=true
    Module.Write SpriteSize=true
    Position=&SimulationPosition()
    Sprite Rotation=&UniformRangedFloat(Minimum=0, Maximum=360)
    Sprite Size=&Vector2DFromFloat(Value=&UniformRangedFloat(Minimum=15, Maximum=5))
   )
   /SetVariables(Particles.AlphaFadeIn:NiagaraFloat=&Min_Float(A:NiagaraFloat=&Divide_Float(A=$Emitter.Age, B=1), B:NiagaraFloat=1)) {isSetParameters=true}
   /SetVariables(Particles.SpriteSize=&Multiply_Vector2(
     A=$Particles.SpriteSize
     B=&Vector2DFromFloat(Value=&OneMinusFloat(Float=$Emitter.Scalability.DistanceFraction))
    )) {isSetParameters=true}
   &AddVelocity(Velocity=&FN_FrontEndVector3Modifier(Gameplay=<0 400 0>, Front End=<0 -400 0>)) {module=AddVelocity001}
   &SkeletalMeshLocation(
    Bone / Socket Index=$Particles.SpawnGroup
    Bone Sampling Mode=::NewEnumerator4
    MaskSpawnGroup=false
    Sampled Position Offset=&MaskVectorBySpawnGroup(Masked Vector=<-30 0 0>, Passthrough Vector=<30 0 0>)
    Skeletal Mesh=$User.SkeletalMesh
   ) {enabled=false, module=SkeletalMeshLocation001}
   &ShapeLocation(
    MaskSpawnGroup=false
    Offset=&MaskVectorBySpawnGroup(
     Masked Vector=&MakeCustomVectorFromBool(Bool=$User.bIsFrontEndPreview, True Vector=<-190 0 -35>, False Vector=<-190 0 -19>)
     Passthrough Vector=&MakeCustomVectorFromBool(Bool=$User.bIsFrontEndPreview, True Vector=<190 0 -35>, False Vector=<190 0 -19>)
    )
    Offset Mode=::NewEnumerator0
    Override Seed=false
    Sphere Radius=10
    UseOwnerScale=true
   ) {module=ShapeLocation001}
  ]
  particleUpdateScript [
   &Color(
    Color=&ColorFromCurve(ColorCurve=DataInterface() {propertyValues {
       AlphaCurve {keys [<at 0 0 RCIM_Cubic>,<at 0.2951 0.06 RCIM_Cubic>,<at 0.9829 0>]}
       BlueCurve {keys [<at 0 0 RCIM_Cubic>]}
       GreenCurve {keys [<at 0 0.574 RCIM_Cubic>]}
       RedCurve {keys [<at 0 1 RCIM_Cubic>]}
      }})
    Scale Alpha=&MakeCustomFloatFromBool(Bool=$User.bIsFrontEndPreview, True Float=0.4, False Float=$Particles.AlphaFadeIn)
   )
   &ParticleState()
   &SolveForcesAndVelocity(Clamp Velocity=false, Limit Acceleration=false)
   &ScaleSpriteSize(Scale Factor=&Vector2DFromFloat(Value=&FloatFromCurve(FloatCurve=DataInterface() {propertyValues {Curve {keys [<at 0 1.6>,<at 1 1>]}}}, Scale Curve=3.5)))
  ]
  rendererClasses [&NiagaraSpriteRendererProperties]
  renderers [{rendererClass=&NiagaraSpriteRendererProperties}]
 }
}
summary {
 systemName=NS_Cadet_Dark_Bright_Glider_Trail
 emitters [
  {emitterName=SpeedLine, rendererClasses [&NiagaraSpriteRendererProperties]}
  {emitterName=StaticHandSprites001, rendererClasses [&NiagaraSpriteRendererProperties]}
  {emitterName=AnimTrail, rendererClasses [&NiagaraRibbonRendererProperties]}
 ]
 userVariables [
  User.bIsFrontEnd {type=NiagaraBool}
  User.bIsFrontEndPreview {type=NiagaraBool}
  User.bIsFullyDeployed {type=NiagaraBool}
  User.FEVec {type=Vector3f, description="", defaultValue {struct=Vector3f, value=<0 0 0>}}
  User.ForwardDot {type=NiagaraFloat, defaultValue=0, description=""}
  User.IsDebugInEditor {type=NiagaraBool}
  User.PlayerSpeed {type=NiagaraFloat, defaultValue=0, description=""}
  User.RibbonIDOffset {type=NiagaraInt32}
  User.RightDot {type=NiagaraFloat, defaultValue=0, description=""}
  User.SkeletalMesh {
   type=NiagaraDataInterfaceSkeletalMesh
   defaultValue {
    struct=NiagaraExt_VariableValue_DataInterface
    value {dataInterface=&self:NiagaraDataInterfaceSkeletalMesh_0}
   }
  }
 ]
}
